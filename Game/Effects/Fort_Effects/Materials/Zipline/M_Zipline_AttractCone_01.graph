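# Translucent, unlit, two-sided mesh-particle material for the zipline attract cone effect, previewed on a cylinder beam mesh.
# Particle color blends toward a hue-shifted copy along UV V, feeding emissive adjustments.
# Opacity combines four blended panning caustic samples with gradient, Fresnel and axis-aligned Fresnel masks, depth fade (skipped on mobile) and synced distance fade.
# A panning, gradient-masked distortion normal is scaled by Dynamic Parameter Param3, while PanSpeed drives all panners.
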Material /Game/Effects/Fort_Effects/Materials/Zipline/M_Zipline_AttractCone_01
graph {
 properties {
  bUsedWithMeshParticles=true
  blendMode=BLEND_Translucent
  pixelDepthOffsetMode=PDOM_Legacy
  previewMesh=/Game/Effects/Fort_Effects/Mesh/Beams/SM_Cylinder_Beam
  shadingModel=MSM_Unlit
  twoSided=true
  usedShadingModels=Unlit
 }
 nodes [
  n1 = ParticleColor()
  n2 = ScalarParameter() {defaultValue=0.25, parameterName="Hue Shift Percentage (S)"}
  n3 = ComponentMask(None=TextureCoordinate()) {g=true}
  n4 = Clamp(None=Power(Base=OneMinus(None=@n3)))
  MaterialFunctionCall(Emissive=LinearInterpolate(
    A=@n1.RGB
    B=MaterialFunctionCall(Hue Shift Percentage=@n2, Texture=@n1.RGB) {materialFunction=&HueShift}.Result
    Alpha=@n4
   )) {materialFunction=&MF_FX_EmissiveAdjustments}
  n5 = Clamp(None=Power(Base=Fresnel() {baseReflectFraction=0.001, exponent=0.01}) {constExponent=8})
  n6 = LinearInterpolate(A=Constant(), B=Constant() {r=1}, Alpha=Multiply(A=@n4, B=@n3))
  n7 = Multiply(A=Time(), B=DynamicParameter() {paramNames [ScaleOffset,PanSpeed,Param3,Param4]}.PanSpeed)
  n8 = TextureSample(UVs=Panner(Coordinate=TextureCoordinate() {uTiling=3, vTiling=0.25}, Time=@n7) {speedX=-0.5, speedY=1}) {texture=&T_ShieldCaustics}
  n9 = TextureSample(UVs=Panner(Coordinate=TextureCoordinate() {uTiling=6, vTiling=0.5}, Time=@n7) {speedX=1, speedY=1}) {texture=&T_ShieldCaustics}
  n10 = TextureSample(UVs=Panner(Coordinate=TextureCoordinate() {uTiling=4, vTiling=0.5}, Time=@n7) {speedX=-0.5, speedY=1}) {texture=&T_ShieldCaustics}
  n11 = TextureSample(UVs=Panner(Coordinate=TextureCoordinate() {uTiling=8}, Time=@n7) {speedX=1, speedY=1}) {texture=&T_ShieldCaustics}
  n12 = Power(Base=LinearInterpolate(A=LinearInterpolate(A=@n8.G, B=@n9.G), B=LinearInterpolate(A=@n10.G, B=@n11.G))) {constExponent=0.5}
  n13 = Multiply(A=ParticleColor().A, B=Multiply(A=OneMinus(None=@n5), B=Multiply(A=@n6, B=@n12)))
  n14 = ShadingPathSwitch(Default=DepthFade(Opacity=@n13, FadeDistance=ScalarParameter() {defaultValue=250, parameterName=FadeDistance}), Mobile=@n13) {desc=Migrated from FeatureLevelSwitch}
  n15 = Fresnel()
  n16 = Multiply(
   A=Power(Base=OneMinus(None=@n15)) {constExponent=16}
   B=MaterialFunctionCall(
    Axis=VectorParameter() {parameterName="Axis (V3)", defaultValue=<rgba 0 0 1 0>}.RGB
    Exponent=ScalarParameter() {defaultValue=1, parameterName="Exponent (S)"}
   ) {materialFunction=&AxisAlignedFresnel}.Fresnel
  )
  n17 = ComponentMask(None=TextureCoordinate()) {g=true}
  MaterialFunctionCall(
   Opacity In=Multiply(A=Multiply(A=@n14, B=@n16), B=Power(Base=@n17))
   Fade_Start_Distance=ScalarParameter() {defaultValue=256, parameterName="Fade_Start_Distance (S)"}
   Fade_End_Distance=ScalarParameter() {defaultValue=128, parameterName="Fade_End_Distance (S)"}
  ) {materialFunction=&MF_Distance_SyncedFade_FX}
  n18 = VectorParameter() {parameterName=Distort_Panner_Tiler, defaultValue=<rgba 0 1 0 1>}
  n19 = Panner(
   Coordinate=Multiply(A=ComponentMask(None=@n18.RGB) {g=true, r=true}, B=TextureCoordinate())
   Time=Multiply(A=Time(), B=DynamicParameter() {paramNames [ScaleOffset,PanSpeed,Param3,Param4]}.PanSpeed)
  ) {speedY=1}
  n20 = TextureCoordinate()
  n21 = OneMinus(None=Multiply(A=MaterialFunctionCall(Float2=@n20) {materialFunction=&BreakOutFloat2Components}.G) {constB=1.25})
  n22 = MaterialFunctionCall(Float2=TextureCoordinate() {vTiling=0.8}) {materialFunction=&BreakOutFloat2Components}
  Multiply(
   A=DynamicParameter() {paramNames [ScaleOffset,PanSpeed,Param3,Param4]}.Param3
   B=LinearInterpolate(
    A=Constant()
    B=TextureSample(UVs=@n19) {samplerType=SAMPLERTYPE_Normal, texture=&T_Weap_Distort_N_02}.RGB
    Alpha=Multiply(A=@n21, B=@n22.G)
   )
  )
  ScalarParameter() {defaultValue=1, parameterName=Speed}
  VectorParameter() {parameterName="Normal (V3)", defaultValue=<rgba 0 0 1 0>}
 ]
}
