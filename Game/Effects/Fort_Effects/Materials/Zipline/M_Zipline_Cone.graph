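# Translucent, unlit, two-sided mesh-particle material for the zipline cone effect, drawing scrolling energy caustics.
# Emissive layers three rotated, panned caustic samples at different tilings and speeds, tinted by particle color and time of day.
# Opacity combines Fresnel edge softening, V gradients, particle alpha and camera, distance and depth fades.
# A panning normal map, masked along V and scaled by a dynamic parameter, adds wave distortion; mobile drops it and depth fade.
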Material /Game/Effects/Fort_Effects/Materials/Zipline/M_Zipline_Cone
graph {
 properties {
  bEnableResponsiveAA=true
  bUsedWithMeshParticles=true
  blendMode=BLEND_Translucent
  pixelDepthOffsetMode=PDOM_Legacy
  shadingModel=MSM_Unlit
  twoSided=true
  usedShadingModels=Unlit
 }
 nodes [
  n1 = Multiply(A=TextureCoordinate() {uTiling=4}, B=DynamicParameter().Param1)
  n2 = Multiply(A=Time(), B=DynamicParameter().Param2)
  n3 = ScalarParameter() {defaultValue=0.25, parameterName="Rotation Angle (0-1) (S)"}
  n4 = TextureSample(UVs=MaterialFunctionCall(UVs=Panner(Coordinate=@n1, Time=@n2) {speedY=0.35}, "Rotation Angle (0-1)"=@n3) {materialFunction=&CustomRotator}.Rotated Values) {texture=&T_ShieldCaustics}
  n5 = TextureCoordinate() {uTiling=2}
  n6 = DynamicParameter()
  n7 = MaterialFunctionCall(UVs=Panner(Coordinate=Multiply(A=@n5, B=@n6.Param1), Time=@n2) {speedY=0.5}, "Rotation Angle (0-1)"=@n3) {materialFunction=&CustomRotator}
  n8 = TextureCoordinate() {uTiling=3}
  n9 = DynamicParameter()
  n10 = MaterialFunctionCall(UVs=Panner(Coordinate=Multiply(A=@n8, B=@n9.Param1), Time=@n2) {speedY=0.7}, "Rotation Angle (0-1)"=@n3) {materialFunction=&CustomRotator}
  n11 = LinearInterpolate(
   A=ScalarParameter() {defaultValue=0.15, parameterName=A}
   B=ScalarParameter() {defaultValue=1, parameterName=B}
   Alpha=Multiply(
    A=TextureSample(UVs=@n7.Rotated Values) {texture=&T_ShieldCaustics}.G
    B=TextureSample(UVs=@n10.Rotated Values) {texture=&T_ShieldCaustics}.G
   )
  )
  n12 = ParticleColor()
  n13 = MaterialFunctionCall(
   Color_Morning=VectorParameter() {parameterName="Color_Morning (V3)", defaultValue=<rgba 1 1 1 1>}.RGB
   Color_Day=VectorParameter() {parameterName="Color_Day (V3)", defaultValue=<rgba 0.75 0.75 0.75 1>}.RGB
   Color_Evening=VectorParameter() {parameterName="Color_Evening (V3)", defaultValue=<rgba 1.1 1.1 1.1 1.1>}.RGB
   Color_Night=VectorParameter() {parameterName="Color_Night (V3)", defaultValue=<rgba 1.25 1.25 1.25 1.25>}.RGB
  ) {materialFunction=&Color_Time_of_Day}
  MaterialFunctionCall(Emissive=Multiply(A=Add(A=@n4.G, B=@n11), B=Multiply(A=@n12.RGB, B=@n13.Result))) {materialFunction=&MF_FX_EmissiveAdjustments}
  n14 = ScalarParameter() {defaultValue=256, parameterName="Distance (S)"}
  n15 = TextureCoordinate()
  n16 = Power(Base=OneMinus(None=Fresnel())) {constExponent=16}
  n17 = MaterialFunctionCall() {materialFunction=&LinearGradient}
  n18 = Multiply(
   A=MaterialFunctionCall(
    Axis=VectorParameter() {parameterName="Axis (V3)", defaultValue=<rgba 0 0 1 0>}.RGB
    Exponent=ScalarParameter() {defaultValue=1, parameterName="Exponent (S)"}
   ) {materialFunction=&AxisAlignedFresnel}.Fresnel
   B=Multiply(A=@n17.VGradient, B=OneMinus(None=@n17.VGradient))
  )
  n19 = Multiply(
   A=LinearInterpolate(A=Constant(), B=Constant() {r=1}, Alpha=ComponentMask(None=@n15) {g=true})
   B=MaterialFunctionCall(
    Opacity In=Multiply(A=@n16, B=@n18)
    Fade_Start_Distance=ScalarParameter() {defaultValue=256, parameterName="Fade_Start_Distance (S)"}
    Fade_End_Distance=ScalarParameter() {defaultValue=128, parameterName="Fade_End_Distance (S)"}
   ) {materialFunction=&MF_Distance_SyncedFade_FX}.Result
  )
  n20 = ParticleColor()
  n21 = ScalarParameter() {defaultValue=256, parameterName=FadeDistance}
  Multiply(
   A=Multiply(
    A=MaterialFunctionCall(Distance=@n14) {materialFunction=&Near_Camera_Fade}.Result
    B=Multiply(A=@n19, B=@n20.A)
   )
   B=ShadingPathSwitch(Default=DepthFade(FadeDistance=@n21), Mobile=Constant() {r=1}) {desc=Migrated from FeatureLevelSwitch}
  )
  n22 = Constant()
  n23 = ComponentMask(None=VectorParameter() {parameterName=Distort_Panner_Tiler, defaultValue=<rgba 0 1 0 1>}.RGB) {g=true, r=true}
  n24 = TextureCoordinate()
  n25 = Time()
  n26 = ScalarParameter() {defaultValue=1, parameterName=Speed}
  n27 = TextureSample(UVs=Panner(Coordinate=Multiply(A=@n23, B=@n24), Time=Multiply(A=@n25, B=@n26)) {speedY=1}) {samplerType=SAMPLERTYPE_Normal, texture=&T_Weap_Distort_N_02}
  n28 = MaterialFunctionCall(Float2=TextureCoordinate()) {materialFunction=&BreakOutFloat2Components}
  n29 = Multiply(
   A=OneMinus(None=Multiply(A=@n28.G))
   B=MaterialFunctionCall(Float2=TextureCoordinate() {vTiling=0.8}) {materialFunction=&BreakOutFloat2Components}.G
  )
  ShadingPathSwitch(
   Default=Multiply(
    A=DynamicParameter() {defaultValue=<rgba 1 50 250 1>, paramNames [Glow, WaveIntensity_Bot, WaveIntensity_Top, Param4]}.WaveIntensity_Bot
    B=LinearInterpolate(A=@n22, B=@n27.RGB, Alpha=@n29)
   )
   Mobile=Constant()
  ) {desc=Migrated from FeatureLevelSwitch}
 ]
}
